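# Material function for billboard impostors that samples a 3x3 baked atlas and outputs base color, normal, masks, UVs and WPO.
# WPO either collapses back-facing quads or, via a static switch, expands camera-facing sprites with virtual-plane UVs.
# Static switches select bump-offset or multi-step POM parallax, clamped per frame.
# View-angle masks drive dithered opacity between planes, and ShadowReplace substitutes sun direction for camera rays in shadow passes.
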MaterialFunction /ImpostorBaker/MaterialFunctions/Impostor_Billboard
graph {nodes [
  CameraVectorWS()
  Constant() {r=1}
  FunctionOutput(None=TextureSample(UVs=NamedRerouteUsage(), Tex=NamedRerouteUsage()).RGBA) {
   description="Contains RGBA value from the optional PackedMasks input. Could be used for things like Specular, Roughness, OpacityMask etc"
   outputName=Optional Masks
   sortPriority=5
  }
  n1 = TextureSample(UVs=NamedRerouteUsage(), Tex=NamedRerouteUsage())
  FunctionOutput(None=@n1.RGB) {description=BaseColor value of Impostor sampling., outputName=BaseColor}
  n2 = TextureSample(UVs=NamedRerouteUsage(), Tex=NamedRerouteUsage())
  FunctionOutput(None=Transform(None=ConstantBiasScale(None=@n2.RGB) {bias=-0.5, scale=2}) {transformSourceType=TRANSFORMSOURCE_Local}) {description=Normal map value of Impostor sampling., outputName=Normal, sortPriority=2}
  FunctionOutput(None=NamedRerouteUsage()) {description=World Position Offset from Impostor sampling., outputName=WPO, sortPriority=3}
  FunctionOutput(None=OneMinus(None=@n1.A)) {outputName=Scatter Mask, sortPriority=5}
  n3 = Multiply(A=Multiply(A=Round(None=@n1.A), B=NamedRerouteUsage()), B=NamedRerouteUsage())
  n4 = NamedRerouteUsage()
  FunctionOutput(None=Add(A=Multiply(A=@n3, B=@n4), B=ScalarParameter() {parameterName=MaskOffset})) {description=Opacity Mask from Impostor sampling., outputName=Opacity Mask, sortPriority=1}
  FunctionOutput(None=NamedRerouteUsage()) {outputName=Billboard UVs, sortPriority=6}
  n5 = MaterialFunctionCall() {materialFunction=/ImpostorBaker/MaterialFunctions/Internal/ImpostorInfo}
  Multiply(A=@n5.Scaled Size, B=Multiply(A=NamedRerouteUsage(), B=NamedRerouteUsage()))
  n6 = NamedRerouteUsage()
  n7 = NamedRerouteUsage()
  NamedRerouteDeclaration(None=StaticSwitchParameter(True=@n6, False=StaticSwitchParameter(True=@n6, False=@n7) {parameterName=BumpOffset Parallax}) {parameterName=Multi Iteration Parallax}) {name=UVs Final, nodeColor=<rgba 1 0.9141 0 1>}
  n8 = Multiply(A=ComponentMask(None=TextureCoordinate()) {r=true}) {constB=0.99}
  NamedRerouteDeclaration(None=Saturate(None=Floor(None=@n8))) {name=Top Mask, nodeColor=<rgba 1 0 0.9609 1>}
  NamedRerouteDeclaration(None=FunctionInput(Preview=TextureObject() {samplerType=SAMPLERTYPE_LinearColor, texture=/Engine/Functions/Engine_MaterialFunctions02/PivotPainter2/Black_1x1_EXR_Texture}) {inputName=Normal, inputType=FunctionInput_Texture2D, sortPriority=2}) {name=NormalTex, nodeColor=<rgba 0.0937 0.4331 1 1>}
  n9 = Transform(None=VertexNormalWS()) {transformSourceType=TRANSFORMSOURCE_World, transformType=TRANSFORM_Local}
  n10 = Transform(None=NamedRerouteUsage()) {transformSourceType=TRANSFORMSOURCE_World, transformType=TRANSFORM_Local}
  n11 = DotProduct(A=Normalize(VectorInput=ComponentMask(None=@n9) {g=true, r=true}), B=Normalize(VectorInput=ComponentMask(None=@n10) {g=true, r=true})) {desc=View around Z Cosine}
  n12 = MaterialFunctionCall!(
   Alpha=NamedRerouteDeclaration(None=Saturate(None=@n11)) {name=Cosine Around Z Axis, nodeColor=<rgba 1 0.2578 0 1>}
   Offset=ScalarParameter() {defaultValue=0.75, group=Vertex Shader, parameterName=Backface Side Threshold}
  ) {materialFunction=/Engine/Functions/Engine_MaterialFunctions01/Opacity/AlphaOffset}
  n13 = ScalarParameter() {defaultValue=2, group=Dithering, parameterName=Dither Side Exponent}
  n14 = NamedRerouteUsage()
  n15 = Saturate(None=ComponentMask(None=Normalize(VectorInput=@n10)) {b=true, desc=B is Cosine of Top View Angle})
  n16 = NamedRerouteDeclaration(None=@n15) {name=Cosine Along Z Axis, nodeColor=<rgba 0 0.4375 1 1>}
  n17 = MaterialFunctionCall(
   Alpha=NamedRerouteDeclaration(None=MaterialFunctionCall!(Side 1 Length=@n16) {desc="Cosine of  Angle from Perpendicular", materialFunction=/Engine/Functions/Engine_MaterialFunctions02/Math/pythagoreanTheorum}.Side 2 Length) {name=Cosine from Perpendicular, nodeColor=<rgba 0 0.6719 1 1>}
   Offset=ScalarParameter() {defaultValue=0.25, group=Vertex Shader, parameterName=Backface Top Threshold}
  ) {materialFunction=&AlphaOffset}
  n18 = ScalarParameter() {defaultValue=0.5, group=Dithering, parameterName=Dither Top Exponent}
  n19 = NamedRerouteUsage()
  n20 = Min(
   A=Power(Base=VertexInterpolator(VS=@n12.Result).PS, Exp=LinearInterpolate(A=@n13, Alpha=@n14) {constB=0})
   B=Power(Base=VertexInterpolator(VS=@n17.Result).PS, Exp=LinearInterpolate(A=@n18, Alpha=@n19) {constB=0})
  )
  n21 = ScalarParameter() {defaultValue=0.25, group=Vertex Shader, parameterName=Backface Top Threshold}
  n22 = Power(
   Base=VertexInterpolator(VS=MaterialFunctionCall(Alpha=@n15, Offset=@n21) {materialFunction=&AlphaOffset}.Result).PS
   Exp=LinearInterpolate(A=ScalarParameter() {defaultValue=0.5, group=Dithering, parameterName=Dither Top Exponent}, Alpha=NamedRerouteUsage()) {constB=0}
  )
  n23 = NamedRerouteUsage()
  n24 = Constant() {r=1}
  n25 = ScalarParameter() {defaultValue=1.5, parameterName=Shadow Dithering}
  NamedRerouteDeclaration(None=Power(Base=LinearInterpolate(A=@n20, B=@n22, Alpha=@n23), Exp=ShadowReplace(Default=@n24, Shadow=@n25))) {name=View Angle Dithering, nodeColor=<rgba 0 0.1328 1 1>}
  n26 = Constant() {r=1}
  n27 = Multiply(A=TextureCoordinate()) {constB=3}
  n28 = Constant()
  n29 = If(
   A=MaterialFunctionCall!(2D Index=Floor(None=@n27), Columns X=Constant() {r=3}) {materialFunction=/Engine/Functions/Engine_MaterialFunctions01/Coordinates/2Dto1DIndex}.1D Index
   "A > B"=@n28
   "A == B"=Constant() {r=1}
   "A < B"=@n28
  )
  n30 = ScalarParameter() {group=Debug, parameterName=Enable Isolation}
  NamedRerouteDeclaration(None=StaticSwitchParameter(True=LinearInterpolate(A=@n26, B=@n29, Alpha=@n30), False=Constant() {r=1}) {group=Debug, parameterName=Debug Single Frame}) {name=Isolate Frame Mask, nodeColor=<rgba 1 0.1875 0 1>}
  n31 = ScalarParameter() {defaultValue=0.5, group=Billboard, parameterName=Reference Plane Sides}
  n32 = ScalarParameter() {defaultValue=0.5, group=Billboard, parameterName=Reference Plane Top}
  n33 = NamedRerouteUsage()
  n34 = MaterialFunctionCall!(
   Heightmap Texture=NamedRerouteUsage()
   Height Ratio=Divide(A=ScalarParameter() {defaultValue=1, group=Parallax, parameterName=Enable Parallax, sliderMax=1}) {constB=3}
   Min Steps
   Max Steps=ScalarParameter() {defaultValue=64, group=Parallax, parameterName=MaxSteps}
   UVs=NamedRerouteUsage()
   Heightmap Channel=Constant4Vector() {constant=<rgba 0 0 0 1>}.RGBA
   Reference Plane=NamedRerouteDeclaration(None=LinearInterpolate(A=@n31, B=@n32, Alpha=@n33)) {name=Reference Plane, nodeColor=<rgba 0 1 0.1562 1>}
   -----------------------------
   Use World Coordinates
   Specify Manual Texture Size
   ClampMinMax=NamedRerouteUsage()
   Manual Texture Size
   Ray Direction=NamedRerouteUsage()
  ) {materialFunction=/ImpostorBaker/MaterialFunctions/Internal/POM_Impostor}
  NamedRerouteDeclaration(None=@n34.Intersection Mask) {name=Parallax Mask, nodeColor=<rgba 1 0 0.5391 1>}
  n35 = Multiply(A=TextureCoordinate()) {constB=3}
  n36 = Divide(A=Floor(None=@n35)) {constB=3}
  NamedRerouteDeclaration(None=@n36) {name=Frame Offset, nodeColor=<rgba 0 1 0.8828 1>}
  n37 = NamedRerouteUsage()
  n38 = Constant() {r=1}
  NamedRerouteDeclaration(None=ShadowReplace(Default=StaticSwitchParameter(True=@n37, False=@n38) {parameterName=Multi Iteration Parallax}, Shadow=Constant() {r=1})) {name=Parallax Mask Final, nodeColor=<rgba 0.3828 1 0 1>}
  n39 = NamedRerouteUsage()
  n40 = ScalarParameter() {defaultValue=0.75, group=Vertex Shader, parameterName=Backface Side Threshold}
  n41 = NamedRerouteUsage()
  n42 = ScalarParameter() {defaultValue=0.25, group=Vertex Shader, parameterName=Backface Top Threshold}
  n43 = LinearInterpolate(
   A=Min(A=Subtract(A=@n39, B=@n40), B=Subtract(A=@n41, B=@n42))
   B=Subtract(A=NamedRerouteUsage(), B=ScalarParameter() {defaultValue=0.25, group=Vertex Shader, parameterName=Backface Top Threshold})
   Alpha=NamedRerouteUsage()
  )
  NamedRerouteDeclaration(None=Saturate(None=Ceil(None=@n43))) {name=BackFace Mask, nodeColor=<rgba 0.3359 1 0 1>}
  n44 = MaterialFunctionCall() {materialFunction=&TangentBasis}
  n45 = WorldPosition()
  n46 = NamedRerouteUsage()
  n47 = Add(A=NamedRerouteUsage(), B=NamedRerouteUsage())
  n48 = MaterialFunctionCall(Float2=MaterialFunctionCall!(
    Plane X Axis=Multiply(A=@n44.X U) {constB=-1}
    Plane Y Axis
    Plane Normal Axis=VertexNormalWS()
    UV Scale=MaterialFunctionCall() {materialFunction=&ImpostorInfo}.Scaled Size
    Ray Dir Local=Subtract(A=@n45.XYZ, B=@n46)
    Ray Origin Local=Subtract(A=@n46, B=@n47)
   ) {materialFunction=/ImpostorBaker/MaterialFunctions/Internal/VirtualPlaneCoordinates_Imposter}.UVs) {materialFunction=&BreakOutFloat2Components}
  n49 = AppendVector(A=OneMinus(None=@n48.R), B=@n48.G)
  NamedRerouteDeclaration(None=Add(A=Divide(A=@n49) {constB=3}, B=NamedRerouteUsage())) {name=Virtual Plane UVs, nodeColor=<rgba 0 1 0.0625 1>}
  n50 = WorldPosition()
  n51 = OneMinus(None=Multiply(A=NamedRerouteUsage(), B=NamedRerouteUsage()))
  n52 = Normalize(VectorInput=Transform(None=VertexNormalWS()) {transformSourceType=TRANSFORMSOURCE_Local})
  n53 = Multiply(
   A=Subtract(A=NamedRerouteUsage()) {constB=0.5}
   B=MaterialFunctionCall() {materialFunction=&ImpostorInfo}.Scaled Size
  )
  n54 = Add(
   A=Subtract(A=Constant(), B=Multiply(A=@n50.XYZ, B=@n51))
   B=NamedRerouteDeclaration(None=Multiply(A=@n52, B=@n53)) {name=Reference Plane Offset, nodeColor=<rgba 0.2891 1 0 1>}
  )
  n55 = NamedRerouteUsage()
  NamedRerouteDeclaration(None=StaticSwitchParameter(True=NamedRerouteUsage(), False=Subtract(A=@n54, B=@n55)) {desc="When using Sprite VS,  the UVs need to be virtually generated", parameterName=Use Sprite Vertex Shader}) {name=WPO, nodeColor=<rgba 0 0.4657 1 1>}
  NamedRerouteDeclaration(None=StaticSwitchParameter(True=NamedRerouteUsage(), False=TextureCoordinate()) {desc="When using Sprite VS,  the UVs need to be virtually generated", parameterName=Use Sprite Vertex Shader}) {name=UV, nodeColor=<rgba 0.6172 1 0 1>}
  NamedRerouteDeclaration(None=TransformPosition(None=VectorParameter() {parameterName=Pivot Offset}.RGB) {transformType=TRANSFORMPOSSOURCE_World}) {name=Billboard Center, nodeColor=<rgba 1 0 0.6562 1>}
  NamedRerouteDeclaration(None=MaterialFunctionCall() {materialFunction=&DitherTemporalAA}.Result) {name=Dither Pattern, nodeColor=<rgba 0 1 0.9766 1>}
  NamedRerouteDeclaration(None=FunctionInput(Preview=TextureObject() {samplerType=SAMPLERTYPE_LinearColor, texture=&Black_1x1_EXR_Texture}) {inputName=BaseColor, inputType=FunctionInput_Texture2D, sortPriority=1}) {name=BaseColorTex, nodeColor=<rgba 0.0937 0.4331 1 1>}
  n56 = Ceil(None=DotProduct(A=NamedRerouteUsage(), B=VertexNormalWS()))
  NamedRerouteDeclaration(None=StaticSwitchParameter(True=Saturate(None=@n56), False=Constant() {r=1}) {desc="When using Sprite VS,  the UVs need to be virtually generated", parameterName=Use Sprite Vertex Shader}) {name=Virtual Plane Mask, nodeColor=<rgba 0.2891 1 0 1>}
  n57 = NamedRerouteUsage()
  n58 = CameraPositionWS()
  n59 = NamedRerouteUsage()
  n60 = Normalize(VectorInput=ShadowReplace(Default=Subtract(A=@n58, B=@n59), Shadow=SkyAtmosphereLightDirection()))
  n61 = VertexNormalWS()
  n62 = NamedRerouteUsage()
  n63 = OneMinus(None=Saturate(None=DotProduct(A=@n61, B=@n62)))
  n64 = ScalarParameter() {defaultValue=8, group=Vertex Shader, parameterName=Backfacing Quad Bias}
  n65 = NamedRerouteDeclaration(None=Multiply(A=NamedRerouteDeclaration(None=@n60) {name=RayDirection, nodeColor=<rgba 0.1484 0 1 1>}, B=Multiply(A=@n63, B=@n64))) {name=Backfacing Sort Bias, nodeColor=<rgba 0 0.1328 1 1>}
  n66 = Multiply(A=TextureCoordinate()) {constB=3}
  n67 = Add(A=ScalarParameter() {defaultValue=1, group=Vertex Shader, parameterName=Sprite Expansion})
  NamedRerouteDeclaration(None=MaterialFunctionCall!(
    Center Location=Subtract(A=@n57, B=@n65)
    0-1 UVs=Frac(None=@n66)
    XY scale=Multiply(A=@n5.Scaled Size, B=@n67)
    Ray Origin=NamedRerouteUsage()
    "Normals (Optional)"
   ) {materialFunction=/ImpostorBaker/MaterialFunctions/Internal/Sprite_PositionBased}.World position offset) {name=Sprite WPO, nodeColor=<rgba 0.7812 1 0 1>}
  n68 = NamedRerouteUsage()
  n69 = ScalarParameter() {defaultValue=1, group=Parallax, parameterName=Enable Parallax, sliderMax=1}
  NamedRerouteDeclaration(None=MaterialFunctionCall!(
    HeightTexture=TextureSample(Tex=@n68).A
    TextureCoordinates=NamedRerouteUsage()
    ReferencePlane=NamedRerouteUsage()
    HeightRatioInput=Divide(A=@n69) {constB=3}
    CameraVector
   ) {materialFunction=/Engine/Functions/Engine_MaterialFunctions02/Utility/BumpOffset_Advanced.BumpOffset_advanced}.UVs) {name=Bump Offset UVs, nodeColor=<rgba 0.4766 0 1 1>}
  n70 = NamedRerouteUsage()
  n71 = NamedRerouteUsage()
  NamedRerouteDeclaration(None=ShadowReplace(Default=StaticSwitchParameter(True=@n70, False=@n71) {parameterName=Multi Iteration Parallax}, Shadow=NamedRerouteUsage())) {name=UVs Parallax Choice, nodeColor=<rgba 1 0 0.3984 1>}
  n72 = Multiply(A=SkyAtmosphereLightDirection()) {constA=100000, constB=100000}
  n73 = NamedRerouteUsage()
  n74 = NamedRerouteDeclaration(None=ShadowReplace(Default=CameraPositionWS(), Shadow=Add(A=@n72, B=@n73))) {name=Ray Origin, nodeColor=<rgba 1 0 0.3984 1>}
  n75 = WorldPosition()
  NamedRerouteDeclaration(None=Normalize(VectorInput=Subtract(A=@n74, B=@n75.XYZ))) {name=CameraVector Derived, nodeColor=<rgba 1 0 0.6094 1>}
  n76 = FunctionInput(Preview=TextureObject() {samplerType=SAMPLERTYPE_LinearColor, texture=&Black_1x1_EXR_Texture}) {
   description=" Could be used for things like Specular, Roughness, OpacityMask etc."
   inputName=ScalarMaps
   inputType=FunctionInput_Texture2D
   sortPriority=4
  }
  NamedRerouteDeclaration(None=@n76) {name=ScalarMapsTex, nodeColor=<rgba 0.0937 0.4331 1 1>}
  n77 = NamedRerouteUsage()
  NamedRerouteDeclaration(None=Clamp(None=NamedRerouteUsage(), Min=ComponentMask(None=@n77) {g=true, r=true}, Max=ComponentMask(None=@n77) {a=true, b=true})) {name=UV Clamped, nodeColor=<rgba 0.9922 0 1 1>}
  n78 = Ceil(None=@n35)
  NamedRerouteDeclaration(None=AppendVector(A=@n36, B=Divide(A=@n78) {constB=3})) {name=Ray Bounds, nodeColor=<rgba 0.1016 1 0 1>}
  NamedRerouteDeclaration(None=@n34.Parallax UVs) {name=Parallax UVs, nodeColor=<rgba 0.0597 0.3863 1 1>}
  NamedRerouteUsage()
  NamedRerouteUsage()
 ]}
