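# Material function for Impostor Baker billboard impostors that layers optional Emissive, Tangent and Subsurface texture maps onto incoming material attributes.
# Three chained static switch parameters each either write a map's RGB into EmissiveColor, Tangent or SubsurfaceColor, or pass the attributes through unchanged.
# All three texture parameters share the Billboard UVs input through a named reroute, which falls back to a default texture coordinate.
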MaterialFunction /ImpostorBaker/MaterialFunctions/Impostor_Billboard_OptionalColorMaps
graph {nodes [
  n1 = TextureSampleParameter2D(UVs=NamedRerouteUsage()) {parameterName=Subsurface, texture=null}
  n2 = NamedRerouteUsage()
  n3 = NamedRerouteUsage()
  n4 = FunctionInput() {inputName=Attributes, inputType=FunctionInput_MaterialAttributes}
  n5 = StaticSwitchParameter(
   True=SetMaterialAttributes!(MaterialAttributes=@n4, EmissiveColor=TextureSampleParameter2D(UVs=@n3) {parameterName=Emissive, texture=null}.RGB) {attributeSetTypes [#1]}
   False=@n4
  ) {parameterName=UseEmissiveTexture}
  n6 = StaticSwitchParameter(
   True=SetMaterialAttributes!(MaterialAttributes=@n5, Tangent=TextureSampleParameter2D(UVs=@n2) {parameterName=Tangent, texture=null}.RGB) {attributeSetTypes [#2]}
   False=@n5
  ) {parameterName=UseTangentTexture}
  FunctionOutput(None=StaticSwitchParameter(
    True=SetMaterialAttributes!(MaterialAttributes=@n6, SubsurfaceColor=@n1.RGB) {attributeSetTypes [#3]}
    False=@n6
   ) {parameterName=UseSubsurfaceTexture})
  NamedRerouteDeclaration(None=FunctionInput(Preview=TextureCoordinate()) {bUsePreviewValueAsDefault=true, inputName=Billboard UVs, inputType=FunctionInput_Vector2, sortPriority=1}) {name=UVs, nodeColor=<rgba 0.1554 1 0 1>}
 ]}
