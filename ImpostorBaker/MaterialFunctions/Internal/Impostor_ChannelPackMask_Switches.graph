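# Impostor Baker material function that writes channels of a PackedMasks input into Material Attributes.
# A chain of static switches optionally overrides Metallic, Specular, Roughness, Anisotropy, Opacity, Opacity Mask, CustomData0/1 for clear coat and its roughness, and Ambient Occlusion.
# Each attribute picks R, G, B or A through nested static switch parameters, defaulting to R.
# Disabled attributes pass through unchanged, and all selection resolves at compile time.
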MaterialFunction /ImpostorBaker/MaterialFunctions/Internal/Impostor_ChannelPackMask_Switches
graph {nodes [
  n1 = MaterialFunctionCall(Float4=NamedRerouteUsage()) {materialFunction=&BreakOutFloat4Components}
  n2 = StaticSwitchParameter(True=@n1.A, False=Constant()) {group=Channel Packed Masks, parameterName=AmbientOcclusion_Channel_A}
  n3 = StaticSwitchParameter(
   True=@n1.R
   False=StaticSwitchParameter(True=@n1.G, False=StaticSwitchParameter(True=@n1.B, False=@n2) {group=Channel Packed Masks, parameterName=AmbientOcclusion_Channel_B}) {group=Channel Packed Masks, parameterName=AmbientOcclusion_Channel_G}
  ) {defaultValue=true, group=Channel Packed Masks, parameterName=AmbientOcclusion_Channel_R}
  n4 = MaterialFunctionCall(Float4=NamedRerouteUsage()) {materialFunction=&BreakOutFloat4Components}
  n5 = Constant()
  n6 = StaticSwitchParameter(
   True=@n4.G
   False=StaticSwitchParameter(True=@n4.B, False=StaticSwitchParameter(True=@n4.A, False=@n5) {group=Channel Packed Masks, parameterName=ClearCoatRoughness_Channel_A}) {group=Channel Packed Masks, parameterName=ClearCoatRoughness_Channel_B}
  ) {group=Channel Packed Masks, parameterName=ClearCoatRoughness_Channel_G}
  n7 = MaterialFunctionCall(Float4=NamedRerouteUsage()) {materialFunction=&BreakOutFloat4Components}
  n8 = Constant()
  n9 = StaticSwitchParameter(
   True=@n7.G
   False=StaticSwitchParameter(True=@n7.B, False=StaticSwitchParameter(True=@n7.A, False=@n8) {group=Channel Packed Masks, parameterName=ClearCoat_Channel_A}) {group=Channel Packed Masks, parameterName=ClearCoat_Channel_B}
  ) {group=Channel Packed Masks, parameterName=ClearCoat_Channel_G}
  n10 = MaterialFunctionCall(Float4=NamedRerouteUsage()) {materialFunction=&BreakOutFloat4Components}
  n11 = Constant()
  n12 = StaticSwitchParameter(
   True=@n10.G
   False=StaticSwitchParameter(True=@n10.B, False=StaticSwitchParameter(True=@n10.A, False=@n11) {group=Channel Packed Masks, parameterName=OpacityMask_Channel_A}) {group=Channel Packed Masks, parameterName=OpacityMask_Channel_B}
  ) {group=Channel Packed Masks, parameterName=OpacityMask_Channel_G}
  n13 = MaterialFunctionCall(Float4=NamedRerouteUsage()) {materialFunction=&BreakOutFloat4Components}
  n14 = Constant()
  n15 = StaticSwitchParameter(
   True=@n13.G
   False=StaticSwitchParameter(True=@n13.B, False=StaticSwitchParameter(True=@n13.A, False=@n14) {group=Channel Packed Masks, parameterName=Opacity_Channel_A}) {group=Channel Packed Masks, parameterName=Opacity_Channel_B}
  ) {group=Channel Packed Masks, parameterName=Opacity_Channel_G}
  n16 = MaterialFunctionCall(Float4=NamedRerouteUsage()) {materialFunction=&BreakOutFloat4Components}
  n17 = Constant()
  n18 = StaticSwitchParameter(
   True=@n16.G
   False=StaticSwitchParameter(True=@n16.B, False=StaticSwitchParameter(True=@n16.A, False=@n17) {group=Channel Packed Masks, parameterName=Anisotropy_Channel_A}) {group=Channel Packed Masks, parameterName=Anisotropy_Channel_B}
  ) {group=Channel Packed Masks, parameterName=Anisotropy_Channel_G}
  n19 = MaterialFunctionCall(Float4=NamedRerouteUsage()) {materialFunction=&BreakOutFloat4Components}
  n20 = Constant()
  n21 = StaticSwitchParameter(
   True=@n19.G
   False=StaticSwitchParameter(True=@n19.B, False=StaticSwitchParameter(True=@n19.A, False=@n20) {group=Channel Packed Masks, parameterName=Roughness_Channel_A}) {group=Channel Packed Masks, parameterName=Roughness_Channel_B}
  ) {group=Channel Packed Masks, parameterName=Roughness_Channel_G}
  n22 = MaterialFunctionCall(Float4=NamedRerouteUsage()) {materialFunction=&BreakOutFloat4Components}
  n23 = Constant()
  n24 = StaticSwitchParameter(
   True=@n22.G
   False=StaticSwitchParameter(True=@n22.B, False=StaticSwitchParameter(True=@n22.A, False=@n23) {group=Channel Packed Masks, parameterName=Specular_Channel_A}) {group=Channel Packed Masks, parameterName=Specular_Channel_B}
  ) {group=Channel Packed Masks, parameterName=Specular_Channel_G}
  n25 = MaterialFunctionCall(Float4=NamedRerouteUsage()) {materialFunction=&BreakOutFloat4Components}
  n26 = Constant()
  n27 = StaticSwitchParameter(
   True=@n25.G
   False=StaticSwitchParameter(True=@n25.B, False=StaticSwitchParameter(True=@n25.A, False=@n26) {group=Channel Packed Masks, parameterName=Metallic_Channel_A}) {group=Channel Packed Masks, parameterName=Metallic_Channel_B}
  ) {group=Channel Packed Masks, parameterName=Metallic_Channel_G}
  n28 = FunctionInput() {inputType=FunctionInput_MaterialAttributes}
  n29 = StaticSwitchParameter(
   True=SetMaterialAttributes!(
    MaterialAttributes=@n28
    Metallic=StaticSwitchParameter(True=@n25.R, False=@n27) {defaultValue=true, group=Channel Packed Masks, parameterName=Metallic_Channel_R}
   ) {attributeSetTypes [#1]}
   False=@n28
  ) {group=Channel Packed Masks, parameterName=UseMetallicTexture}
  n30 = StaticSwitchParameter(
   True=SetMaterialAttributes!(
    MaterialAttributes=@n29
    Specular=StaticSwitchParameter(True=@n22.R, False=@n24) {defaultValue=true, group=Channel Packed Masks, parameterName=Specular_Channel_R}
   ) {attributeSetTypes [#2]}
   False=@n29
  ) {group=Channel Packed Masks, parameterName=UseSpecularTexture}
  n31 = StaticSwitchParameter(
   True=SetMaterialAttributes!(
    MaterialAttributes=@n30
    Roughness=StaticSwitchParameter(True=@n19.R, False=@n21) {defaultValue=true, group=Channel Packed Masks, parameterName=Roughness_Channel_R}
   ) {attributeSetTypes [#3]}
   False=@n30
  ) {group=Channel Packed Masks, parameterName=UseRoughnessTexture}
  n32 = StaticSwitchParameter(
   True=SetMaterialAttributes!(
    MaterialAttributes=@n31
    Anisotropy=StaticSwitchParameter(True=@n16.R, False=@n18) {defaultValue=true, group=Channel Packed Masks, parameterName=Anisotropy_Channel_R}
   ) {attributeSetTypes [#4]}
   False=@n31
  ) {group=Channel Packed Masks, parameterName=UseAnisotropyTexture}
  n33 = StaticSwitchParameter(
   True=SetMaterialAttributes!(
    MaterialAttributes=@n32
    Opacity=StaticSwitchParameter(True=@n13.R, False=@n15) {defaultValue=true, group=Channel Packed Masks, parameterName=Opacity_Channel_R}
   ) {attributeSetTypes [#5]}
   False=@n32
  ) {group=Channel Packed Masks, parameterName=UseOpacityTexture}
  n34 = StaticSwitchParameter(
   True=SetMaterialAttributes!(
    MaterialAttributes=@n33
    OpacityMask=StaticSwitchParameter(True=@n10.R, False=@n12) {defaultValue=true, group=Channel Packed Masks, parameterName=OpacityMask_Channel_R}
   ) {attributeSetTypes [#6]}
   False=@n33
  ) {group=Channel Packed Masks, parameterName=UseOpacityMaskTexture}
  n35 = StaticSwitchParameter(
   True=SetMaterialAttributes!(
    MaterialAttributes=@n34
    CustomData0=StaticSwitchParameter(True=@n7.R, False=@n9) {defaultValue=true, group=Channel Packed Masks, parameterName=ClearCoat_Channel_R}
   ) {attributeSetTypes [#7]}
   False=@n34
  ) {group=Channel Packed Masks, parameterName=UseClearCoatTexture}
  n36 = StaticSwitchParameter(
   True=SetMaterialAttributes!(
    MaterialAttributes=@n35
    CustomData1=StaticSwitchParameter(True=@n4.R, False=@n6) {defaultValue=true, group=Channel Packed Masks, parameterName=ClearCoatRoughness_Channel_R}
   ) {attributeSetTypes [#8]}
   False=@n35
  ) {group=Channel Packed Masks, parameterName=UseClearCoatRoughnessTexture}
  FunctionOutput(None=StaticSwitchParameter(
    True=SetMaterialAttributes!(MaterialAttributes=@n36, AmbientOcclusion=@n3) {attributeSetTypes [#9]}
    False=@n36
   ) {group=Channel Packed Masks, parameterName=UseAmbientOcclusionTexture}) {outputName=Attributes}
  NamedRerouteDeclaration(None=FunctionInput() {bUsePreviewValueAsDefault=true, inputName=PackedMasks, inputType=FunctionInput_Vector4, sortPriority=1}) {name=PackedMasks, nodeColor=<rgba 0.1729 0.4997 1 1>}
 ]}
