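# Parent material for billboard impostors in ImpostorBaker: masked two-sided foliage with non-tangent-space normals, usable on instanced static meshes.
# The Impostor_Billboard function turns base color, normal and packed-mask textures into surface attributes, billboard UVs and masks.
# A remapped scatter mask scales base color and SubsurfaceColor into an edge glow.
# Scalar parameters set specular, roughness and opacity before functions apply optional color maps and channel-packed mask switches.
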Material /ImpostorBaker/Materials/Parents/M_Billboard
graph {
 properties {
  bTangentSpaceNormal=false
  bUseMaterialAttributes=true
  bUsedWithInstancedStaticMeshes=true
  bUsedWithStaticLighting=true
  blendMode=BLEND_Masked
  numCustomizedUVs=1
  opacityMaskClipValue=0.5
  pixelDepthOffsetMode=PDOM_Legacy
  shadingModel=MSM_TwoSidedFoliage
  translucencyLightingMode=TLM_SurfacePerPixelLighting
  usedShadingModels=Two Sided Foliage
 }
 nodes [
  n1 = MaterialFunctionCall!(
   BaseColor=TextureObjectParameter() {parameterName=BaseColor, texture=&Black}
   Normal=TextureObjectParameter() {parameterName=Normal, texture=&Black}
   ScalarMaps=TextureObjectParameter() {desc=Optional Custom Color Map, parameterName=PackedMasks_1, texture=&Black}
  ) {materialFunction=/ImpostorBaker/MaterialFunctions/Impostor_Billboard}
  n2 = SetMaterialAttributes!(
   MaterialAttributes
   Base Color=@n1.BaseColor
   Opacity Mask=@n1.BaseColor
   Normal=@n1.BaseColor
   World Position Offset=@n1.BaseColor
   Pixel Depth Offset
   Roughness
  ) {attributeSetTypes [
    #1
    #2
    #3
    #4
    #5
    #6
   ]}
  n3 = GetMaterialAttributes(None=@n2) {attributeGetTypes [#1]}
  n4 = VectorParameter() {parameterName=SubsurfaceColor, defaultValue=<rgba 1 1 1 1>}
  n5 = Subtract(A=NamedRerouteUsage(), B=ScalarParameter() {defaultValue=0.1, desc=">> Start of inner edge for edge glow", parameterName=Scatter Mask Min})
  n6 = ScalarParameter() {defaultValue=2, desc=">> Width of gradient for edge glow", parameterName=Scatter Mask Length}
  n7 = Multiply(A=@n4.A, B=Saturate(None=Divide(A=@n5, B=@n6)))
  n8 = SetMaterialAttributes!(
   MaterialAttributes=@n3.MaterialAttributes
   Specular=ScalarParameter() {defaultValue=0.5, parameterName=Specular}
   Roughness=ScalarParameter() {defaultValue=0.5, parameterName=Roughness}
   Opacity=ScalarParameter() {parameterName=Opacity}
   Subsurface Color=Multiply(A=@n3.Base Color, B=Multiply(A=@n4.RGB, B=@n7))
  ) {attributeSetTypes [
    #7
    #6
    #8
    #9
   ]}
  MaterialFunctionCall!(
   In=MaterialFunctionCall!(Attributes=@n8, Billboard UVs=NamedRerouteUsage()) {materialFunction=/ImpostorBaker/MaterialFunctions/Impostor_Billboard_OptionalColorMaps}.Result
   PackedMasks=NamedRerouteUsage()
  ) {materialFunction=/ImpostorBaker/MaterialFunctions/Internal/Impostor_ChannelPackMask_Switches}
  NamedRerouteDeclaration(None=@n1.Scatter Mask) {name=Scatter Mask, nodeColor=<rgba 1 0.4688 0 1>}
  NamedRerouteDeclaration(None=@n1.Optional Masks) {name=Packed Masks, nodeColor=<rgba 0.1248 0.4423 1 1>}
  NamedRerouteDeclaration(None=@n1.Billboard UVs) {name=Billboard UVs, nodeColor=<rgba 0.1997 0.0281 1 1>}
  VertexNormalWS()
 ]
}
