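# Top-level ImpostorBaker material function that renders baked impostors, outputting color, normal, opacity mask, WPO, PDO and optional masks.
# Impostor_UVs_and_WPO supplies three UV planes and weights for blended texture sampling, with a single-frame fallback via static switch or mobile override.
# WPO adds shadow bias and a camera-height-scaled mobile offset; PDO is zeroed for orthographic views and low/medium quality.
# Material Attributes serve as a generic struct, passing blend data downstream for cheaper extra sampling.
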MaterialFunction /ImpostorBaker/MaterialFunctions/Impostor
graph {nodes [
  Composite() {subgraphName=Mobile Z Offset}
  FunctionOutput(None=MaterialFunctionCall!(
    Texture Object=NamedRerouteUsage()
    R UV=NamedRerouteUsage()
    G UV=NamedRerouteUsage()
    B UV=NamedRerouteUsage()
    Weights=NamedRerouteUsage()
    Derivative UVs=NamedRerouteUsage()
   ) {desc=Normal and Depth, materialFunction=/ImpostorBaker/MaterialFunctions/Internal/FrameBlendWeights_TextureObject}.Result) {
   description="Contains RGBA value from the optional PackedMasks input. Could be used for things like Specular, Roughness, OpacityMask etc"
   outputName=Optional Masks
   sortPriority=5
  }
  n1 = MaterialFunctionCall(
   Texture Object=NamedRerouteUsage()
   R UV=NamedRerouteUsage()
   G UV=NamedRerouteUsage()
   B UV=NamedRerouteUsage()
   Weights=NamedRerouteUsage()
   Derivative UVs=NamedRerouteUsage()
  ) {desc=Color and Alpha, materialFunction=&FrameBlendWeights_TextureObject}
  n2 = NamedRerouteUsage()
  n3 = MaterialFunctionCall!(Default=ComponentMask(None=@n1.Result) {b=true, g=true, r=true}, Mobile Override=@n2) {materialFunction=/ImpostorBaker/MaterialFunctions/Internal/OverrideMobile}
  FunctionOutput(None=LinearInterpolate(
    A=StaticSwitchParameter(True=@n3.Result, False=@n2) {defaultValue=true, parameterName=Blend Multiple Frames}
    B=NamedRerouteUsage()
    Alpha=NamedRerouteUsage()
   )) {description=BaseColor value of Impostor sampling., outputName=BaseColor}
  n4 = MaterialFunctionCall(
   Texture Object=NamedRerouteUsage()
   R UV=NamedRerouteUsage()
   G UV=NamedRerouteUsage()
   B UV=NamedRerouteUsage()
   Weights=NamedRerouteUsage()
   Derivative UVs=NamedRerouteUsage()
  ) {desc=Normal and Depth, materialFunction=&FrameBlendWeights_TextureObject}
  n5 = ComponentMask(None=@n4.Result) {b=true, g=true, r=true}
  n6 = NamedRerouteUsage()
  n7 = ConstantBiasScale(None=StaticSwitchParameter(
    True=MaterialFunctionCall(Default=@n5, Mobile Override=@n6) {materialFunction=&OverrideMobile}.Result
    False=@n6
   ) {defaultValue=true, parameterName=Blend Multiple Frames}) {bias=-0.5, scale=2}
  n8 = NamedRerouteUsage()
  FunctionOutput(None=LinearInterpolate(A=Transform(None=@n7), B=Constant3Vector() {constant=<rgba 0 0 1 0>}, Alpha=Saturate(None=@n8))) {description=Normal map value of Impostor sampling., outputName=Normal, sortPriority=2}
  n9 = Subtract(
   A=ComponentMask(None=CameraPositionWS()) {b=true}
   B=ScalarParameter() {defaultValue=10000, group=Mobile Z Offset, parameterName=Camera Z Fadeout Loc}
  )
  n10 = ScalarParameter() {defaultValue=35000, group=Mobile Z Offset, parameterName=Cam Z Fade Length}
  n11 = Multiply(A=VectorParameter() {parameterName=MobileZOffset, defaultValue=<rgba 0 0 550 0>}.RGB, B=Saturate(None=Divide(A=@n9, B=@n10)))
  MaterialExpressionReroute_6 = Reroute(None=@n11)
  n12 = Add(A=@MaterialExpressionReroute_6.Output, B=NamedRerouteUsage())
  n13 = NamedRerouteUsage()
  n14 = StaticSwitchParameter(
   True=MaterialFunctionCall(Default=NamedRerouteUsage(), Mobile Override=DistanceFieldsRenderingSwitch!(No=@n12, Yes=@n13)) {materialFunction=&OverrideMobile}.Result
   False=NamedRerouteUsage()
  ) {defaultValue=true, parameterName=Blend Multiple Frames}
  n15 = NamedRerouteDeclaration(None=Multiply(A=SkyAtmosphereLightDirection(), B=ScalarParameter() {group=Shadowing, parameterName=Shadow Bias})) {name=Shadow Bias, nodeColor=<rgba 1 0.2109 0 1>}
  FunctionOutput(None=ShadowReplace(Default=@n14, Shadow=Subtract(A=@n14, B=@n15))) {description=World Position Offset from Impostor sampling., outputName=WPO, sortPriority=3}
  n16 = OneMinus(None=IsOrthographic())
  n17 = NamedRerouteUsage()
  n18 = StaticSwitchParameter(True=NamedRerouteUsage(), False=@n17) {
   defaultValue=true
   desc=A fully accurate version of Pixel Depth Offset that reconstructs the worldposition of each of the 3 interpolated frames using its original transform.
   parameterName=Use Depth Derived PDO
  }
  n19 = StaticSwitchParameter(
   True=MaterialFunctionCall(
    Default=StaticSwitchParameter(True=@n18, False=@n17) {desc=Use Depth Derived PDO when using Iterative Parallax, parameterName=Multi Iteration Parallax}
    Mobile Override=@n17
   ) {materialFunction=&OverrideMobile}.Result
   False=NamedRerouteUsage()
  ) {defaultValue=true, parameterName=Blend Multiple Frames}
  n20 = Constant()
  FunctionOutput(None=QualitySwitch(Default=Multiply(A=@n16, B=@n19), Low=@n20, Medium=@n20)) {description=Pixel Depth Offset from impostor sampling., outputName=PDO, sortPriority=4}
  n21 = NamedRerouteUsage()
  n22 = MaterialFunctionCall(Default=ComponentMask(None=ShadowReplace(Default=@n1.Result, Shadow=@n1.Result)) {a=true}, Mobile Override=@n21) {materialFunction=&OverrideMobile}
  n23 = Add(
   A=ComponentMask(None=StaticSwitchParameter(True=@n22.Result, False=@n21) {defaultValue=true, parameterName=Blend Multiple Frames}) {r=true}
   B=NamedRerouteUsage()
  ) {desc=Add Debug}
  n24 = NamedRerouteUsage()
  FunctionOutput(None=Add(
    A=Multiply(A=@n23, B=@n24) {desc="Clip debug box\r\nNeeded due to WPO overscan when debugging\r\nBecause the mesh cuts out the unused space"}
    B=ScalarParameter() {parameterName=MaskOffset}
   )) {description=Opacity Mask from Impostor sampling., outputName=Opacity Mask, sortPriority=1}
  n25 = Constant()
  n26 = SetMaterialAttributes!(
   MaterialAttributes
   BaseColor=AppendVector(A=NamedRerouteUsage(), B=@n25)
   EmissiveColor=AppendVector(A=NamedRerouteUsage(), B=@n25)
   SubsurfaceColor=AppendVector(A=NamedRerouteUsage(), B=@n25)
   WorldPositionOffset=AppendVector(A=NamedRerouteUsage(), B=@n25)
   Normal=NamedRerouteUsage()
  ) {
   desc="Note: Attributes used as generic struct.\r\nValues to not match their label.\r\nIe BaseColor here contains UVs."
   attributeSetTypes [
    #1
    #2
    #3
    #4
    #5
   ]
  }
  FunctionOutput(None=@n26) {
   description=Uses Material Attributes to pass along necessary blend information in order to sample additional textures downstream without paying for the setup cost again.
   outputName=Impostor Data
   sortPriority=6
  }
  n27 = FunctionInput() {bUsePreviewValueAsDefault=true, inputName=Frames XY, inputType=FunctionInput_Scalar, previewValue=<16 0 0 1>}
  n28 = MaterialFunctionCall!(
   Frames XY=NamedRerouteDeclaration(None=@n27) {name=Frames XY, nodeColor=<rgba 1 0 0.4922 1>}
   BaseColor=NamedRerouteUsage()
   Normal=NamedRerouteUsage()
   ScalarMaps=NamedRerouteUsage()
  ) {materialFunction=/ImpostorBaker/MaterialFunctions/Internal/Impostor_UVs_and_WPO}
  NamedRerouteDeclaration(None=@n28.Depth Derived PDO) {name=Depth Derived PDO, nodeColor=<rgba 1 0.9141 0 1>}
  NamedRerouteDeclaration(None=FunctionInput(Preview=TextureObject() {samplerType=SAMPLERTYPE_LinearColor, texture=/Engine/Functions/Engine_MaterialFunctions02/PivotPainter2/Black_1x1_EXR_Texture}) {inputName=Normal, inputType=FunctionInput_Texture2D, sortPriority=2}) {name=NormalTex, nodeColor=<rgba 0.0937 0.4331 1 1>}
  n29 = MaterialFunctionCall!(Frames XY=NamedRerouteUsage(), BaseColor=NamedRerouteUsage(), Normal=NamedRerouteUsage()) {materialFunction=&/Impostor_SingleFrameVersion}
  NamedRerouteDeclaration(None=@n29.BaseColor RGB) {name=Single Frame Color, nodeColor=<rgba 0 0 0 1>}
  NamedRerouteDeclaration(None=@n29.BaseColor A) {name=Single Frame Alpha, nodeColor=<rgba 0 0 0 1>}
  NamedRerouteDeclaration(None=@n29.Normal RGB) {name=Single Frame Normal, nodeColor=<rgba 0 0 0 1>}
  n30 = NamedRerouteUsage()
  n31 = ComponentMask(None=MaterialFunctionCall(Default=ComponentMask(None=@n4.Result) {a=true}, Mobile Override=@n30) {materialFunction=&OverrideMobile}.Result) {r=true}
  NamedRerouteDeclaration(None=OneMinus(None=StaticSwitchParameter(True=@n31, False=@n30) {defaultValue=true, parameterName=Blend Multiple Frames})) {name=PDO Local, nodeColor=<rgba 0.5937 1 0 1>}
  NamedRerouteDeclaration(None=@n28.Blend Weights) {name=Blend Weights, nodeColor=<rgba 0.0937 0.4331 1 1>}
  NamedRerouteDeclaration(None=Multiply(A=TextureCoordinate() {uTiling=10, vTiling=10}, B=ScalarParameter() {defaultValue=0.6, parameterName=Mip UV Scale})) {name=Derivative UVs, nodeColor=<rgba 0.0937 0.4331 1 1>}
  NamedRerouteDeclaration(None=@n28.Debug Color) {name=Frame Debug Display, nodeColor=<rgba 0 0 0 1>}
  NamedRerouteDeclaration(None=ComponentMask(None=@n28.Debug Mask + Clip) {g=true}) {name=Debug Box Clip, nodeColor=<rgba 0 0 0 1>}
  NamedRerouteDeclaration(None=ComponentMask(None=@n28.Debug Mask + Clip) {r=true}) {name=Frame DebugMask, nodeColor=<rgba 0 0 0 1>}
  NamedRerouteDeclaration(None=FunctionInput(Preview=TextureObject() {samplerType=SAMPLERTYPE_LinearColor, texture=&Black_1x1_EXR_Texture}) {inputName=BaseColor, inputType=FunctionInput_Texture2D, sortPriority=1}) {name=BaseColorTex, nodeColor=<rgba 0.0937 0.4331 1 1>}
  n32 = Normalize(VectorInput=CameraVectorWS())
  n33 = MaterialFunctionCall() {materialFunction=&CameraDirectionVector}
  n34 = Multiply(
   A=Abs(None=DotProduct(A=@n32, B=@n33.Result))
   B=MaterialFunctionCall() {materialFunction=/ImpostorBaker/MaterialFunctions/Internal/ImpostorInfo}.Scaled Size
  ) {constB=0.5}
  n35 = ScalarParameter() {defaultValue=1, parameterName=PDO}
  NamedRerouteDeclaration(None=Multiply(A=NamedRerouteUsage(), B=Multiply(A=@n34, B=@n35))) {name=PDO World, nodeColor=<rgba 0.3359 0 1 1>}
  n36 = FunctionInput(Preview=TextureObject() {samplerType=SAMPLERTYPE_LinearColor, texture=&Black_1x1_EXR_Texture}) {
   description=" Could be used for things like Specular, Roughness, OpacityMask etc."
   inputName=ScalarMaps
   inputType=FunctionInput_Texture2D
   sortPriority=4
  }
  NamedRerouteDeclaration(None=@n36) {name=ScalarMapsTex, nodeColor=<rgba 0.0937 0.4331 1 1>}
  NamedRerouteDeclaration(None=@n28.UV Plane R) {name=UV Plane R, nodeColor=<rgba 1 0 0 1>}
  NamedRerouteDeclaration(None=@n28.UV Plane G) {name=UV Plane G, nodeColor=<rgba 0 1 0 1>}
  NamedRerouteDeclaration(None=@n28.UV Plane B) {name=UV Plane B, nodeColor=<rgba 0 0 1 1>}
  NamedRerouteDeclaration(None=@n29.Normal A) {name=SingleFrame PDO, nodeColor=<rgba 0 0 0 1>}
  NamedRerouteDeclaration(None=@n29.WPO) {name=SingleFrame WPO, nodeColor=<rgba 0 0 0 1>}
  NamedRerouteDeclaration(None=@n28.WPO) {name=WorldPositionOffset, nodeColor=<rgba 0.1437 0.4778 1 1>}
  PinBase()
  PinBase!(Output=@n11) {reroutePins [Output {expression=&self:MaterialExpressionReroute_6}]}
 ]}
