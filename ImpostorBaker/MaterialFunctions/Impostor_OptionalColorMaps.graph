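# Material function in the Impostor Baker that adds optional baked color maps to an impostor's material attributes.
# Three static switches, UseEmissiveTexture, UseTangentTexture and UseSubsurfaceTexture, are chained so each can override EmissiveColor, Tangent or SubsurfaceColor on the incoming Attributes, otherwise passing them through.
# Each enabled map samples its texture object parameter through Impostor_ExtraColorMap, fed by a shared Impostor Data input via a named reroute.
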MaterialFunction /ImpostorBaker/MaterialFunctions/Impostor_OptionalColorMaps
graph {nodes [
  n1 = MaterialFunctionCall!(Impostor Data=NamedRerouteUsage(), ColorMap=TextureObjectParameter() {parameterName=Subsurface}) {materialFunction=&/Impostor_ExtraColorMap}
  n2 = NamedRerouteUsage()
  n3 = TextureObjectParameter() {parameterName=Tangent}
  n4 = NamedRerouteUsage()
  n5 = TextureObjectParameter() {parameterName=Emissive}
  n6 = FunctionInput() {inputName=Attributes, inputType=FunctionInput_MaterialAttributes}
  n7 = StaticSwitchParameter(
   True=SetMaterialAttributes!(
    MaterialAttributes=@n6
    EmissiveColor=MaterialFunctionCall(Impostor Data=@n4, ColorMap=@n5) {materialFunction=&Impostor_ExtraColorMap}.ColorMap
   ) {attributeSetTypes [#1]}
   False=@n6
  ) {parameterName=UseEmissiveTexture}
  n8 = StaticSwitchParameter(
   True=SetMaterialAttributes!(
    MaterialAttributes=@n7
    Tangent=MaterialFunctionCall(Impostor Data=@n2, ColorMap=@n3) {materialFunction=&Impostor_ExtraColorMap}.ColorMap
   ) {attributeSetTypes [#2]}
   False=@n7
  ) {parameterName=UseTangentTexture}
  FunctionOutput(None=StaticSwitchParameter(
    True=SetMaterialAttributes!(MaterialAttributes=@n8, SubsurfaceColor=@n1.ColorMap) {attributeSetTypes [#3]}
    False=@n8
   ) {parameterName=UseSubsurfaceTexture})
  NamedRerouteDeclaration(None=FunctionInput() {bUsePreviewValueAsDefault=true, inputName=Impostor Data, inputType=FunctionInput_MaterialAttributes, sortPriority=1}) {name=Impostor Data, nodeColor=<rgba 0.1554 1 0 1>}
 ]}
