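# Translucent unlit parent material for impostor clouds, built on the ImpostorBaker Impostor functions and usable on instanced static meshes.
# Lighting dots sun and skylight directions against four light vectors to weight baked directional lighting, apparently from the CustomLighting map and packed masks.
# Sky-atmosphere sun illuminance with a Schlick phase, distant-scattering skylight and a tinted ground term sum into emissive.
# Opacity comes from a packed channel, and world position offset from the impostor attribute chain.
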Material /ImpostorBaker/Materials/Parents/M_Impostor_Cloud
graph {
 properties {
  bAlwaysEvaluateWorldPositionOffset=true
  bTangentSpaceNormal=false
  bUseMaterialAttributes=true
  bUsedWithInstancedStaticMeshes=true
  bUsedWithStaticLighting=true
  blendMode=BLEND_Translucent
  numCustomizedUVs=1
  opacityMaskClipValue=0.5
  pixelDepthOffsetMode=PDOM_Legacy
  shadingModel=MSM_Unlit
  twoSided=true
  usedShadingModels=Unlit
 }
 nodes [
  n1 = MaterialFunctionCall!(
   Frames XY=ScalarParameter() {defaultValue=16, group=Impostor Setup, parameterName=FramesXY}
   BaseColor=TextureObjectParameter() {parameterName=BaseColor, texture=&Black}
   Normal=TextureObjectParameter() {parameterName=Normal, texture=&Black}
   ScalarMaps=TextureObjectParameter() {desc=Optional Custom Color Map, parameterName=PackedMasks_1, texture=&Black}
  ) {materialFunction=/ImpostorBaker/MaterialFunctions/Impostor}
  NamedRerouteDeclaration(None=@n1.Optional Masks) {name=Packed Masks, nodeColor=<rgba 0.1248 0.4423 1 1>}
  NamedRerouteDeclaration(None=@n1.Impostor Data) {name=Impostor Data, nodeColor=<rgba 0.1997 0.0281 1 1>}
  n2 = SkyAtmosphereLightDirection()
  n3 = Normalize(VectorInput=Transform(None=ComponentMask(None=@n2) {b=true, g=true, r=true}) {transformSourceType=TRANSFORMSOURCE_World, transformType=TRANSFORM_Local})
  n4 = Transform(None=VectorParameter() {parameterName=LightVector_1, defaultValue=<rgba 1 0 0 1>}.RGB) {transformSourceType=TRANSFORMSOURCE_View, transformType=TRANSFORM_Local}
  n5 = Transform(None=VectorParameter() {parameterName=LightVector_2, defaultValue=<rgba -1 0 0 1>}.RGB) {transformSourceType=TRANSFORMSOURCE_View, transformType=TRANSFORM_Local}
  n6 = Transform(None=VectorParameter() {parameterName=LightVector_3, defaultValue=<rgba 0 1 0 1>}.RGB) {transformSourceType=TRANSFORMSOURCE_View, transformType=TRANSFORM_Local}
  n7 = VectorParameter() {parameterName=LightVector_4, defaultValue=<rgba 0 0 1 1>}
  n8 = DotProduct(A=@n3, B=Normalize(VectorInput=Transform(None=@n7.RGB) {transformSourceType=TRANSFORMSOURCE_View, transformType=TRANSFORM_Local}))
  n9 = MaterialFunctionCall(
   X=DotProduct(A=@n3, B=Normalize(VectorInput=@n4))
   Y=DotProduct(A=@n3, B=Normalize(VectorInput=@n5))
   Z=DotProduct(A=@n3, B=Normalize(VectorInput=@n6))
   A=@n8
  ) {materialFunction=&MakeFloat4}
  n10 = Multiply(A=@n8) {constB=-1}
  n11 = Add(A=DotProduct(A=Max(A=@n9.Result) {constB=0}, B=NamedRerouteUsage()), B=Multiply(A=Max(A=@n10) {constB=0}, B=NamedRerouteUsage()))
  n12 = VectorParameter() {parameterName=Light Color, defaultValue=<rgba 1 0.7449 0.4623 1>}
  n13 = AppendVector(
   A=ComponentMask(None=WorldPosition().XYZ) {g=true, r=true}
   B=ScalarParameter() {defaultValue=5000, parameterName=Cloud Altitude for Sun light}
  )
  n14 = DotProduct(A=ComponentMask(None=SkyAtmosphereLightDirection()) {b=true, g=true, r=true}, B=CameraVectorWS())
  n15 = ScalarParameter() {defaultValue=0.4, parameterName=Aniso}
  n16 = Multiply(
   A=Multiply(A=@n12.RGB, B=@n12.RGB)
   B=Multiply(
    A=SkyAtmosphereLightIlluminance!(World Position=@n13)
    B=MaterialFunctionCall!(Cosine=@n14, Anisotropy=@n15) {materialFunction=/Engine/Functions/Engine_MaterialFunctions01/Volumetrics/SchlickPhase}.Result
   )
  )
  n17 = Normalize(VectorInput=Transform(None=VectorParameter() {parameterName=Skylight Vector, defaultValue=<rgba 0 0 1 0>}.RGB) {transformSourceType=TRANSFORMSOURCE_World, transformType=TRANSFORM_Local})
  n18 = VectorParameter() {parameterName=LightVector_1, defaultValue=<rgba 1 0 0 1>}
  n19 = DotProduct(A=@n17, B=Normalize(VectorInput=Transform(None=@n18.RGB) {transformSourceType=TRANSFORMSOURCE_View, transformType=TRANSFORM_Local}))
  n20 = VectorParameter() {parameterName=LightVector_2, defaultValue=<rgba -1 0 0 1>}
  n21 = DotProduct(A=@n17, B=Normalize(VectorInput=Transform(None=@n20.RGB) {transformSourceType=TRANSFORMSOURCE_View, transformType=TRANSFORM_Local}))
  n22 = VectorParameter() {parameterName=LightVector_3, defaultValue=<rgba 0 1 0 1>}
  n23 = DotProduct(A=@n17, B=Normalize(VectorInput=Transform(None=@n22.RGB) {transformSourceType=TRANSFORMSOURCE_View, transformType=TRANSFORM_Local}))
  n24 = VectorParameter() {parameterName=LightVector_4, defaultValue=<rgba 0 0 1 1>}
  n25 = DotProduct(A=@n17, B=Normalize(VectorInput=Transform(None=@n24.RGB) {transformSourceType=TRANSFORMSOURCE_View, transformType=TRANSFORM_Local}))
  n26 = Max(A=DotProduct(
    A=MaterialFunctionCall(X=@n19, Y=@n21, Z=@n23, A=@n25) {materialFunction=&MakeFloat4}.Result
    B=NamedRerouteUsage()
   )) {constB=0}
  n27 = Multiply(A=Max(A=Multiply(A=@n25) {constB=-1}) {constB=0}, B=NamedRerouteUsage())
  n28 = ScalarParameter() {defaultValue=0.075, parameterName=Shadow Lift}
  n29 = Add(A=NamedRerouteDeclaration(None=Add(A=@n26, B=@n27)) {name=Upward Term, nodeColor=<rgba 0 0.5078 1 1>}, B=@n28)
  n30 = Add(A=@n28)
  n31 = SkyAtmosphereDistantLightScatteredLuminance()
  n32 = Multiply(
   A=Max(A=Divide(A=@n29, B=@n30)) {constB=0}
   B=Multiply(A=ComponentMask(None=@n31) {b=true, g=true, r=true}, B=VectorParameter() {parameterName=SkyColor, defaultValue=<rgba 0.5 0.6 1 2>}.A)
  )
  n33 = ScalarParameter() {defaultValue=0.5, parameterName=Ground Lift}
  n34 = Add(A=Multiply(A=NamedRerouteUsage()) {constB=-1}, B=@n33)
  n35 = Add(A=@n33)
  n36 = VectorParameter() {parameterName=GroundColor, defaultValue=<rgba 0.5 0.2 0 0>}
  n37 = Multiply(A=Max(A=Divide(A=@n34, B=@n35)) {constB=0}, B=Multiply(A=@n36.RGB, B=@n36.RGB))
  NamedRerouteDeclaration(None=Add(A=Multiply(A=@n11, B=@n16), B=Add(A=@n32, B=@n37))) {name=Emissive Lighting, nodeColor=<rgba 0 1 0.2969 1>}
  NamedRerouteDeclaration(None=MaterialFunctionCall(Float4=NamedRerouteUsage()) {materialFunction=&BreakOutFloat4Components}.G) {name=Opacity, nodeColor=<rgba 1 0 0.0234 1>}
  n38 = SetMaterialAttributes!(
   MaterialAttributes
   Base Color=@n1.BaseColor
   Opacity Mask=@n1.BaseColor
   Normal=@n1.BaseColor
   World Position Offset=@n1.BaseColor
   Pixel Depth Offset=@n1.BaseColor
   Roughness
  ) {attributeSetTypes [
    #1
    #2
    #3
    #4
    #5
    #6
   ]}
  n39 = GetMaterialAttributes(None=@n38) {attributeGetTypes [#1, #2]}
  n40 = VectorParameter() {parameterName=SubsurfaceColor, defaultValue=<rgba 1 1 1 1>}
  n41 = Subtract(
   A=OneMinus(None=@n39.Opacity Mask)
   B=ScalarParameter() {defaultValue=0.1, desc=">> Start of inner edge for edge glow", parameterName=Scatter Mask Min}
  )
  n42 = ScalarParameter() {defaultValue=2, desc=">> Width of gradient for edge glow", parameterName=Scatter Mask Length}
  n43 = Multiply(A=@n40.A, B=Saturate(None=Divide(A=@n41, B=@n42)))
  n44 = SetMaterialAttributes!(
   MaterialAttributes=@n39.MaterialAttributes
   Specular=ScalarParameter() {defaultValue=0.5, parameterName=Specular}
   Roughness=ScalarParameter() {defaultValue=0.5, parameterName=Roughness}
   Opacity=ScalarParameter() {defaultValue=1, parameterName=Opacity}
   Subsurface Color=Multiply(A=@n39.Base Color, B=Multiply(A=@n40.RGB, B=@n43))
  ) {attributeSetTypes [
    #7
    #6
    #8
    #9
   ]}
  n45 = MaterialFunctionCall!(Attributes=@n44, Impostor Data=NamedRerouteUsage()) {materialFunction=/ImpostorBaker/MaterialFunctions/Impostor_OptionalColorMaps}
  n46 = NamedRerouteUsage()
  NamedRerouteDeclaration(None=GetMaterialAttributes(None=MaterialFunctionCall!(In=@n45.Result, PackedMasks=@n46) {materialFunction=/ImpostorBaker/MaterialFunctions/Internal/Impostor_ChannelPackMask_Switches}.Attributes) {attributeGetTypes [#4]}.World Position Offset) {name=WPO, nodeColor=<rgba 0 0.3437 1 1>}
  n47 = MaterialFunctionCall!(
   Impostor Data=NamedRerouteUsage()
   ColorMap=TextureObjectParameter() {desc=Optional Custom Color Map, parameterName=CustomLighting, texture=&Black}
  ) {materialFunction=/ImpostorBaker/MaterialFunctions/Impostor_ExtraColorMap}
  n48 = ComponentMask(None=@n47.ColorMap) {a=true}
  n49 = AppendVector(A=ComponentMask(None=@n47.ColorMap) {b=true, g=true, r=true}, B=Multiply(A=@n48, B=ScalarParameter() {defaultValue=1, parameterName=ZBoost}))
  NamedRerouteDeclaration(None=DotProduct(A=NamedRerouteDeclaration(None=@n49) {name=Custom Lighting RGBA, nodeColor=<rgba 0 0.3906 1 1>}, B=Constant() {r=0.25})) {name=Combined, nodeColor=<rgba 0 1 0.0859 1>}
  OneMinus(None=@n48)
  OneMinus(None=@n33)
  SetMaterialAttributes!(MaterialAttributes, Emissive Color=NamedRerouteUsage(), Opacity=NamedRerouteUsage(), World Position Offset=NamedRerouteUsage()) {attributeSetTypes [#10, #8, #4]}
 ]
}
